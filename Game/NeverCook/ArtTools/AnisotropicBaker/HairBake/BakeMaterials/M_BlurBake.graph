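# M_BlurBake is an unlit alpha-composite material that blurs a captured texture for the hair anisotropy baking tool.
# A Texture parameter, defaulting to the HairAnisoCapture transient render target, feeds the engine's SpiralBlur-Texture function.
# The BlurRadius parameter, scaled down by a constant divisor, sets the blur distance.
# Distance and radial step counts are fixed high, so each pixel takes many samples.
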Material /Game/NeverCook/ArtTools/AnisotropicBaker/HairBake/BakeMaterials/M_BlurBake
graph {
 properties {
  bUsedWithMeshDeformer=true
  bUsedWithSkeletalMesh=true
  blendMode=BLEND_AlphaComposite
  pixelDepthOffsetMode=PDOM_Legacy
  shadingModel=MSM_Unlit
  usedShadingModels=Unlit
 }
 nodes [
  Constant()
  n1 = TextureObjectParameter() {
   parameterName=Texture
   samplerType=SAMPLERTYPE_LinearColor
   texture=/Game/NeverCook/ArtTools/AnisotropicBaker/HairBake/TransientRendertargets/HairAnisoCapture
  }
  MaterialFunctionCall!(
   TextureObject=@n1
   UV
   Distance=Divide(A=ScalarParameter() {defaultValue=1, parameterName=BlurRadius}) {constB=100}
   Distance Steps=Constant() {r=32}
   Radial Steps=Constant() {r=32}
   Radial Offset
   Kernel Power
  ) {materialFunction=/Engine/Functions/Engine_MaterialFunctions01/Texturing/SpiralBlur-Texture}
 ]
}
