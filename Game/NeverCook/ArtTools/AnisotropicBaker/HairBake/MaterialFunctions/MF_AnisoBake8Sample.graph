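# Material function for the anisotropic hair baker that measures how a texture channel differs from an original sample across eight neighbouring points.
# It calls MF_AnisoBakeSample at four axis and four diagonal offsets, with diagonals shortened to 0.75 per axis, scaled from pixel size.
# The eight AbsDifference results are summed and multiplied by a reciprocal factor to form the output.
# Named reroutes share inputs such as texture, channel, MIP, UV and distance scale across the taps.
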MaterialFunction /Game/NeverCook/ArtTools/AnisotropicBaker/HairBake/MaterialFunctions/MF_AnisoBake8Sample
graph {nodes [
  n1 = NamedRerouteUsage()
  n2 = FunctionInput() {inputName=Texture, inputType=FunctionInput_Texture2D}
  n3 = FunctionInput() {inputName=Channel, inputType=FunctionInput_Vector4}
  n4 = FunctionInput() {inputName=OriginalSample}
  n5 = FunctionInput() {inputName=MIP, inputType=FunctionInput_Scalar}
  n6 = NamedRerouteUsage()
  n7 = FunctionInput() {inputName=PixelSizeOffset, inputType=FunctionInput_Scalar}
  n8 = Multiply(A=Multiply(A=@n7, B=Constant2Vector() {r=1}), B=NamedRerouteUsage())
  n9 = NamedRerouteUsage()
  n10 = NamedRerouteUsage()
  n11 = NamedRerouteUsage()
  n12 = NamedRerouteUsage()
  n13 = NamedRerouteUsage()
  n14 = Multiply(A=Multiply(A=NamedRerouteUsage(), B=Constant2Vector() {g=1}), B=NamedRerouteUsage())
  n15 = NamedRerouteUsage()
  n16 = NamedRerouteUsage()
  n17 = NamedRerouteUsage()
  n18 = NamedRerouteUsage()
  n19 = NamedRerouteUsage()
  n20 = Multiply(A=Multiply(A=NamedRerouteUsage(), B=Constant2Vector() {r=-1}), B=NamedRerouteUsage())
  n21 = NamedRerouteUsage()
  n22 = NamedRerouteUsage()
  n23 = NamedRerouteUsage()
  n24 = NamedRerouteUsage()
  n25 = NamedRerouteUsage()
  n26 = Multiply(A=Multiply(A=NamedRerouteUsage(), B=Constant2Vector() {g=-1}), B=NamedRerouteUsage())
  n27 = Add(
   A=Add(
    A=MaterialFunctionCall!(Texture=@n2, Channel=@n3, OriginalSample=@n4, MIP=@n5, UV=@n6, Offset=@n8) {materialFunction=&/MF_AnisoBakeSample}.AbsDifference
    B=MaterialFunctionCall(Texture=@n9, Channel=@n10, OriginalSample=@n11, MIP=@n12, UV=@n13, Offset=@n14) {materialFunction=&MF_AnisoBakeSample}.AbsDifference
   )
   B=Add(
    A=MaterialFunctionCall(Texture=@n15, Channel=@n16, OriginalSample=@n17, MIP=@n18, UV=@n19, Offset=@n20) {materialFunction=&MF_AnisoBakeSample}.AbsDifference
    B=MaterialFunctionCall(Texture=@n21, Channel=@n22, OriginalSample=@n23, MIP=@n24, UV=@n25, Offset=@n26) {materialFunction=&MF_AnisoBakeSample}.AbsDifference
   )
  )
  n28 = NamedRerouteUsage()
  n29 = NamedRerouteUsage()
  n30 = NamedRerouteUsage()
  n31 = NamedRerouteUsage()
  n32 = NamedRerouteUsage()
  n33 = Multiply(A=Multiply(A=NamedRerouteUsage(), B=Constant2Vector() {g=0.75, r=0.75}), B=NamedRerouteUsage())
  n34 = NamedRerouteUsage()
  n35 = NamedRerouteUsage()
  n36 = NamedRerouteUsage()
  n37 = NamedRerouteUsage()
  n38 = NamedRerouteUsage()
  n39 = Multiply(A=Multiply(A=NamedRerouteUsage(), B=Constant2Vector() {g=0.75, r=-0.75}), B=NamedRerouteUsage())
  n40 = NamedRerouteUsage()
  n41 = NamedRerouteUsage()
  n42 = NamedRerouteUsage()
  n43 = NamedRerouteUsage()
  n44 = NamedRerouteUsage()
  n45 = Multiply(A=Multiply(A=NamedRerouteUsage(), B=Constant2Vector() {g=-0.75, r=0.75}), B=NamedRerouteUsage())
  n46 = NamedRerouteUsage()
  n47 = NamedRerouteUsage()
  n48 = NamedRerouteUsage()
  n49 = NamedRerouteUsage()
  n50 = NamedRerouteUsage()
  n51 = Multiply(A=Multiply(A=NamedRerouteUsage(), B=Constant2Vector() {g=-0.75, r=-0.75}), B=NamedRerouteUsage())
  n52 = Add(
   A=Add(
    A=MaterialFunctionCall(Texture=@n28, Channel=@n29, OriginalSample=@n30, MIP=@n31, UV=@n32, Offset=@n33) {materialFunction=&MF_AnisoBakeSample}.AbsDifference
    B=MaterialFunctionCall(Texture=@n34, Channel=@n35, OriginalSample=@n36, MIP=@n37, UV=@n38, Offset=@n39) {materialFunction=&MF_AnisoBakeSample}.AbsDifference
   )
   B=Add(
    A=MaterialFunctionCall(Texture=@n40, Channel=@n41, OriginalSample=@n42, MIP=@n43, UV=@n44, Offset=@n45) {materialFunction=&MF_AnisoBakeSample}.AbsDifference
    B=MaterialFunctionCall(Texture=@n46, Channel=@n47, OriginalSample=@n48, MIP=@n49, UV=@n50, Offset=@n51) {materialFunction=&MF_AnisoBakeSample}.AbsDifference
   )
  )
  FunctionOutput(None=Multiply(A=Add(A=@n27, B=@n52), B=Divide(B=@n1)))
  NamedRerouteDeclaration(None=@n2) {name=Texture, nodeColor=<rgba 0.6172 0 1 1>}
  NamedRerouteDeclaration(None=@n3) {name=Channel, nodeColor=<rgba 0 1 0.2969 1>}
  NamedRerouteDeclaration(None=@n4) {name=OriginalSample, nodeColor=<rgba 1 0.4688 0 1>}
  NamedRerouteDeclaration(None=@n5) {name=MIP, nodeColor=<rgba 0.7344 1 0 1>}
  NamedRerouteDeclaration(None=@n7) {name=PixelSizeOffset, nodeColor=<rgba 0.8047 1 0 1>}
  NamedRerouteDeclaration(None=FunctionInput() {inputName=SampleDistanceScale, inputType=FunctionInput_Scalar, previewValue=<1 0 0 1>}) {name=SampleDistanceScale, nodeColor=<rgba 0 0.7656 1 1>}
  NamedRerouteDeclaration(None=FunctionInput(Preview=TextureCoordinate()) {inputName=UV, inputType=FunctionInput_Vector2}) {name=inputUVs, nodeColor=<rgba 1 0.0234 0 1>}
 ]}
