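# Art-tool material in NeverCook that bakes an anisotropy direction map for hair from a strand texture.
# It runs MF_AnisoBake8Sample at sixteen increasing distances around texel-snapped UVs on a selectable channel and MIP, then sums and normalizes the results into a 2D direction.
# Sign logic folds opposite directions together unless Method02 selects the raw vector.
# The bake parameter blends in the direction encoded as red-green color; aniso, roughness and normal parameters are also present.
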Material /Game/NeverCook/ArtTools/AnisotropicBaker/HairBake/BakeMaterials/M_HairAnisoBake
graph {
 properties {
  bUsedWithMeshDeformer=true
  bUsedWithMorphTargets=true
  bUsedWithSkeletalMesh=true
  pixelDepthOffsetMode=PDOM_Legacy
  usedShadingModels=Default Lit
 }
 nodes [
  Constant() {r=2048}
  Constant() {r=1}
  n1 = ScalarParameter() {parameterName=bake}
  LinearInterpolate(A=Constant() {r=0.9}, B=Constant(), Alpha=@n1)
  n2 = MaterialFunctionCall!(
   Channel=NamedRerouteUsage()
   OriginalSample=NamedRerouteUsage()
   MIP=NamedRerouteUsage()
   PixelSizeOffset=NamedRerouteUsage()
   SampleDistanceScale=Constant() {r=1}
   UV=NamedRerouteUsage()
   Texture=NamedRerouteUsage()
  ) {materialFunction=/Game/NeverCook/ArtTools/AnisotropicBaker/HairBake/MaterialFunctions/MF_AnisoBake8Sample}
  n3 = MaterialFunctionCall(
   Channel=NamedRerouteUsage()
   OriginalSample=NamedRerouteUsage()
   MIP=NamedRerouteUsage()
   PixelSizeOffset=NamedRerouteUsage()
   SampleDistanceScale=Constant() {r=2}
   UV=NamedRerouteUsage()
   Texture=NamedRerouteUsage()
  ) {materialFunction=&MF_AnisoBake8Sample}
  n4 = MaterialFunctionCall(
   Channel=NamedRerouteUsage()
   OriginalSample=NamedRerouteUsage()
   MIP=NamedRerouteUsage()
   PixelSizeOffset=NamedRerouteUsage()
   SampleDistanceScale=Constant() {r=3}
   UV=NamedRerouteUsage()
   Texture=NamedRerouteUsage()
  ) {materialFunction=&MF_AnisoBake8Sample}
  n5 = MaterialFunctionCall(
   Channel=NamedRerouteUsage()
   OriginalSample=NamedRerouteUsage()
   MIP=NamedRerouteUsage()
   PixelSizeOffset=NamedRerouteUsage()
   SampleDistanceScale=Constant() {r=4}
   UV=NamedRerouteUsage()
   Texture=NamedRerouteUsage()
  ) {materialFunction=&MF_AnisoBake8Sample}
  n6 = MaterialFunctionCall(
   Channel=NamedRerouteUsage()
   OriginalSample=NamedRerouteUsage()
   MIP=NamedRerouteUsage()
   PixelSizeOffset=NamedRerouteUsage()
   SampleDistanceScale=Constant() {r=5}
   UV=NamedRerouteUsage()
   Texture=NamedRerouteUsage()
  ) {materialFunction=&MF_AnisoBake8Sample}
  n7 = MaterialFunctionCall(
   Channel=NamedRerouteUsage()
   OriginalSample=NamedRerouteUsage()
   MIP=NamedRerouteUsage()
   PixelSizeOffset=NamedRerouteUsage()
   SampleDistanceScale=Constant() {r=6}
   UV=NamedRerouteUsage()
   Texture=NamedRerouteUsage()
  ) {materialFunction=&MF_AnisoBake8Sample}
  n8 = MaterialFunctionCall(
   Channel=NamedRerouteUsage()
   OriginalSample=NamedRerouteUsage()
   MIP=NamedRerouteUsage()
   PixelSizeOffset=NamedRerouteUsage()
   SampleDistanceScale=Constant() {r=7}
   UV=NamedRerouteUsage()
   Texture=NamedRerouteUsage()
  ) {materialFunction=&MF_AnisoBake8Sample}
  n9 = MaterialFunctionCall(
   Channel=NamedRerouteUsage()
   OriginalSample=NamedRerouteUsage()
   MIP=NamedRerouteUsage()
   PixelSizeOffset=NamedRerouteUsage()
   SampleDistanceScale=Constant() {r=8}
   UV=NamedRerouteUsage()
   Texture=NamedRerouteUsage()
  ) {materialFunction=&MF_AnisoBake8Sample}
  n10 = Add(
   A=Add(A=Add(A=@n2.Result, B=@n3.Result), B=Add(A=@n4.Result, B=@n5.Result))
   B=Add(A=Add(A=@n6.Result, B=@n7.Result), B=Add(A=@n8.Result, B=@n9.Result))
  )
  n11 = MaterialFunctionCall(
   Channel=NamedRerouteUsage()
   OriginalSample=NamedRerouteUsage()
   MIP=NamedRerouteUsage()
   PixelSizeOffset=NamedRerouteUsage()
   SampleDistanceScale=Constant() {r=9}
   UV=NamedRerouteUsage()
   Texture=NamedRerouteUsage()
  ) {materialFunction=&MF_AnisoBake8Sample}
  n12 = MaterialFunctionCall(
   Channel=NamedRerouteUsage()
   OriginalSample=NamedRerouteUsage()
   MIP=NamedRerouteUsage()
   PixelSizeOffset=NamedRerouteUsage()
   SampleDistanceScale=Constant() {r=10}
   UV=NamedRerouteUsage()
   Texture=NamedRerouteUsage()
  ) {materialFunction=&MF_AnisoBake8Sample}
  n13 = MaterialFunctionCall(
   Channel=NamedRerouteUsage()
   OriginalSample=NamedRerouteUsage()
   MIP=NamedRerouteUsage()
   PixelSizeOffset=NamedRerouteUsage()
   SampleDistanceScale=Constant() {r=11}
   UV=NamedRerouteUsage()
   Texture=NamedRerouteUsage()
  ) {materialFunction=&MF_AnisoBake8Sample}
  n14 = MaterialFunctionCall(
   Channel=NamedRerouteUsage()
   OriginalSample=NamedRerouteUsage()
   MIP=NamedRerouteUsage()
   PixelSizeOffset=NamedRerouteUsage()
   SampleDistanceScale=Constant() {r=12}
   UV=NamedRerouteUsage()
   Texture=NamedRerouteUsage()
  ) {materialFunction=&MF_AnisoBake8Sample}
  n15 = MaterialFunctionCall(
   Channel=NamedRerouteUsage()
   OriginalSample=NamedRerouteUsage()
   MIP=NamedRerouteUsage()
   PixelSizeOffset=NamedRerouteUsage()
   SampleDistanceScale=Constant() {r=13}
   UV=NamedRerouteUsage()
   Texture=NamedRerouteUsage()
  ) {materialFunction=&MF_AnisoBake8Sample}
  n16 = MaterialFunctionCall(
   Channel=NamedRerouteUsage()
   OriginalSample=NamedRerouteUsage()
   MIP=NamedRerouteUsage()
   PixelSizeOffset=NamedRerouteUsage()
   SampleDistanceScale=Constant() {r=14}
   UV=NamedRerouteUsage()
   Texture=NamedRerouteUsage()
  ) {materialFunction=&MF_AnisoBake8Sample}
  n17 = MaterialFunctionCall(
   Channel=NamedRerouteUsage()
   OriginalSample=NamedRerouteUsage()
   MIP=NamedRerouteUsage()
   PixelSizeOffset=NamedRerouteUsage()
   SampleDistanceScale=Constant() {r=15}
   UV=NamedRerouteUsage()
   Texture=NamedRerouteUsage()
  ) {materialFunction=&MF_AnisoBake8Sample}
  n18 = MaterialFunctionCall(
   Channel=NamedRerouteUsage()
   OriginalSample=NamedRerouteUsage()
   MIP=NamedRerouteUsage()
   PixelSizeOffset=NamedRerouteUsage()
   SampleDistanceScale=Constant() {r=16}
   UV=NamedRerouteUsage()
   Texture=NamedRerouteUsage()
  ) {materialFunction=&MF_AnisoBake8Sample}
  n19 = Add(
   A=Add(A=Add(A=@n11.Result, B=@n12.Result), B=Add(A=@n13.Result, B=@n14.Result))
   B=Add(A=Add(A=@n15.Result, B=@n16.Result), B=Add(A=@n17.Result, B=@n18.Result))
  )
  n20 = Normalize(VectorInput=AppendVector(A=Add(A=@n10, B=@n19), B=Constant()))
  n21 = Multiply(A=@n20, B=Constant3Vector() {constant=<rgba -1 -1 1 0>})
  n22 = MaterialFunctionCall(Float3=@n20) {materialFunction=&BreakOutFloat3Components}
  n23 = Sign(None=@n22.R)
  n24 = Saturate(None=@n23)
  n25 = Abs(None=@n20)
  n26 = Sign(None=Multiply(A=@n23, B=Sign(None=@n22.G)))
  n27 = LinearInterpolate(
   A=LinearInterpolate(A=LinearInterpolate(A=@n21, B=@n20, Alpha=@n24), B=@n25, Alpha=Saturate(None=@n26))
   B=@n20
   Alpha=ScalarParameter() {parameterName=Method02}
  )
  LinearInterpolate(B=Multiply(A=ConstantBiasScale(None=@n27), B=Constant3Vector() {constant=<rgba 1 1 0 0>}), Alpha=@n1)
  n28 = ScalarParameter() {defaultValue=2048, parameterName=ImageSize}
  n29 = Divide(A=@n28, B=Add(A=NamedRerouteUsage()))
  NamedRerouteDeclaration(None=Multiply(A=Divide(B=@n29), B=ScalarParameter() {defaultValue=1, parameterName=OffsetInPixels})) {name=PixelSizeOffset, nodeColor=<rgba 0.6172 1 0 1>}
  NamedRerouteDeclaration(None=TextureObjectParameter() {parameterName=StrandMap, texture=/Game/NeverCook/ArtTools/AnisotropicBaker/TestImages/F_MED_HighTower_Honeydew_Hair_Base_color}) {name=Texture, nodeColor=<rgba 0 0.9062 1 1>}
  n30 = NamedRerouteUsage()
  n31 = NamedRerouteUsage()
  n32 = NamedRerouteUsage()
  n33 = VectorParameter() {parameterName=ColorChannel, defaultValue=<rgba 0 1 0 0>}
  n34 = AppendVector(A=@n33.RGB, B=@n33.RGB)
  NamedRerouteDeclaration(None=DotProduct(
    A=TextureSample!(UVs=@n30, Tex=@n31, Level=@n32, Apply View MipBias) {mipValueMode=TMVM_MipLevel}.RGBA
    B=NamedRerouteDeclaration(None=@n34) {name=ColorChannel, nodeColor=<rgba 0.5938 0 1 1>}
   )) {name=OriginalSample, nodeColor=<rgba 1 0 0.5625 1>}
  NamedRerouteDeclaration(None=ScalarParameter() {parameterName=MIP}) {name=MIPLevel, nodeColor=<rgba 1 0 0.3047 1>}
  n35 = TextureCoordinate()
  n36 = Add(A=Ceil(None=Multiply(A=@n35, B=@n28))) {constB=-0.5}
  NamedRerouteDeclaration(None=Saturate(None=Divide(A=@n36, B=@n28))) {name=NearestCoords, nodeColor=<rgba 0 0.6484 1 1>}
  n37 = MaterialFunctionCall(Float3=@n25) {materialFunction=&BreakOutFloat3Components}
  Saturate(None=@n37.R)
  Saturate(None=@n37.G)
  ScalarParameter() {defaultValue=0.8, parameterName=Aniso}
  ScalarParameter() {defaultValue=0.5, parameterName=Roughness}
  TextureCoordinate()
  TextureSampleParameter2D() {parameterName=Normal, samplerType=SAMPLERTYPE_Normal, texture=&FlatNormalBlue}
 ]
}
