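# MF_MaskBlack is a material function in the anisotropic baker's hair bake tools that builds a mask separating black from non-black input.
# It splits a Float3 input, takes the max of red and blue, then applies Ceil, so zero stays 0 and values up to one become 1.
# The first Max compares red with itself, so green is never read and pure green masks as black.
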MaterialFunction /Game/NeverCook/ArtTools/AnisotropicBaker/HairBake/MaterialFunctions/MF_MaskBlack
graph {nodes [
  n1 = MaterialFunctionCall(Float3=FunctionInput()) {materialFunction=&BreakOutFloat3Components}
  n2 = Max(A=@n1.R, B=@n1.R)
  FunctionOutput(None=Ceil(None=Max(A=@n2, B=@n1.B)))
 ]}
