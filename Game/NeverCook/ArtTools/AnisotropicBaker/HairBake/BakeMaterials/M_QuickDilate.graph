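# Dilation material for the hair anisotropy bake art tool, working on the HairAnisoCapture render target through a Texture parameter.
# UVs snapped to 2048-grid texel centres feed one-texel BlurSampleOffsets taps, combined by Max.
# MF_MaskBlack blends the four-neighbour maximum with the original, and an absolute difference against the original isolates the change.
# A wider maximum adds four half-offset diagonal taps; neither branch shows a connection to a material output.
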Material /Game/NeverCook/ArtTools/AnisotropicBaker/HairBake/BakeMaterials/M_QuickDilate
graph {
 properties {pixelDepthOffsetMode=PDOM_Legacy, usedShadingModels=Default Lit}
 nodes [
  n1 = TextureSampleParameter2D!(UVs=NamedRerouteUsage(), Level, Apply View MipBias) {
   mipValueMode=TMVM_MipLevel
   parameterName=Texture
   samplerSource=SSM_Clamp_WorldGroupSettings
   samplerType=SAMPLERTYPE_LinearColor
   texture=/Game/NeverCook/ArtTools/AnisotropicBaker/HairBake/TransientRendertargets/HairAnisoCapture
  }
  n2 = NamedRerouteUsage()
  n3 = Constant() {r=2048}
  n4 = MaterialFunctionCall!(OffsetAmount=Divide(B=@n3)) {materialFunction=/Engine/Functions/Engine_MaterialFunctions01/ImageAdjustment/BlurSampleOffsets}
  n5 = TextureSampleParameter2D(UVs=Saturate(None=Add(A=@n2, B=@n4."0,1"))) {
   mipValueMode=TMVM_MipLevel
   parameterName=Texture
   samplerSource=SSM_Clamp_WorldGroupSettings
   samplerType=SAMPLERTYPE_LinearColor
   texture=&HairAnisoCapture
  }
  n6 = NamedRerouteUsage()
  n7 = TextureSampleParameter2D(UVs=Saturate(None=Add(A=@n6, B=@n4."1,0"))) {
   mipValueMode=TMVM_MipLevel
   parameterName=Texture
   samplerSource=SSM_Clamp_WorldGroupSettings
   samplerType=SAMPLERTYPE_LinearColor
   texture=&HairAnisoCapture
  }
  n8 = NamedRerouteUsage()
  n9 = TextureSampleParameter2D(UVs=Saturate(None=Add(A=@n8, B=@n4."0,-1"))) {
   mipValueMode=TMVM_MipLevel
   parameterName=Texture
   samplerSource=SSM_Clamp_WorldGroupSettings
   samplerType=SAMPLERTYPE_LinearColor
   texture=&HairAnisoCapture
  }
  n10 = NamedRerouteUsage()
  n11 = TextureSampleParameter2D(UVs=Saturate(None=Add(A=@n10, B=@n4."-1,0"))) {
   mipValueMode=TMVM_MipLevel
   parameterName=Texture
   samplerSource=SSM_Clamp_WorldGroupSettings
   samplerType=SAMPLERTYPE_LinearColor
   texture=&HairAnisoCapture
  }
  n12 = Max(A=Max(A=@n5.RGB, B=@n7.RGB), B=Max(A=@n9.RGB, B=@n11.RGB))
  n13 = TextureSampleParameter2D() {
   mipValueMode=TMVM_MipLevel
   parameterName=Texture
   samplerType=SAMPLERTYPE_LinearColor
   texture=&HairAnisoCapture
  }
  n14 = Saturate(None=MaterialFunctionCall!(In=@n13.RGB) {materialFunction=/Game/NeverCook/ArtTools/AnisotropicBaker/HairBake/MaterialFunctions/MF_MaskBlack}.Result)
  Abs(None=Subtract(A=@n1.RGB, B=LinearInterpolate(A=@n12, B=@n1.RGB, Alpha=@n14)))
  n15 = NamedRerouteUsage()
  n16 = TextureSampleParameter2D(UVs=Saturate(None=Add(A=@n15, B=@n4."-.5,.5"))) {
   mipValueMode=TMVM_MipLevel
   parameterName=Texture
   samplerSource=SSM_Clamp_WorldGroupSettings
   samplerType=SAMPLERTYPE_LinearColor
   texture=&HairAnisoCapture
  }
  n17 = NamedRerouteUsage()
  n18 = TextureSampleParameter2D(UVs=Saturate(None=Add(A=@n17, B=@n4.".5,.5"))) {
   mipValueMode=TMVM_MipLevel
   parameterName=Texture
   samplerSource=SSM_Clamp_WorldGroupSettings
   samplerType=SAMPLERTYPE_LinearColor
   texture=&HairAnisoCapture
  }
  n19 = NamedRerouteUsage()
  n20 = TextureSampleParameter2D(UVs=Saturate(None=Add(A=@n19, B=@n4.".5,-.5"))) {
   mipValueMode=TMVM_MipLevel
   parameterName=Texture
   samplerSource=SSM_Clamp_WorldGroupSettings
   samplerType=SAMPLERTYPE_LinearColor
   texture=&HairAnisoCapture
  }
  n21 = NamedRerouteUsage()
  n22 = TextureSampleParameter2D(UVs=Saturate(None=Add(A=@n21, B=@n4."-.5,-.5"))) {
   mipValueMode=TMVM_MipLevel
   parameterName=Texture
   samplerSource=SSM_Clamp_WorldGroupSettings
   samplerType=SAMPLERTYPE_LinearColor
   texture=&HairAnisoCapture
  }
  Max(A=@n12, B=Max(A=Max(A=@n16.RGB, B=@n18.RGB), B=Max(A=@n20.RGB, B=@n22.RGB)))
  n23 = TextureCoordinate()
  n24 = Add(A=Ceil(None=Multiply(A=@n23, B=@n3))) {constB=-0.5}
  NamedRerouteDeclaration(None=Saturate(None=Divide(A=@n24, B=@n3))) {name=NearestCoords, nodeColor=<rgba 0 0.6484 1 1>}
 ]
}
