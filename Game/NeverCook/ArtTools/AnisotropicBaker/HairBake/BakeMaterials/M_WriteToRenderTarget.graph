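# M_WriteToRenderTarget is an unlit, alpha-composite utility material in the anisotropic hair baking art tool, apparently for writing a texture into a render target.
# It holds a linear-color texture parameter named Texture, defaulting to the transient HairDilate01 render target, and a single constant, with their output wiring not shown.
# It sits under NeverCook, which points to editor-time baking rather than runtime rendering.
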Material /Game/NeverCook/ArtTools/AnisotropicBaker/HairBake/BakeMaterials/M_WriteToRenderTarget
graph {
 properties {blendMode=BLEND_AlphaComposite, pixelDepthOffsetMode=PDOM_Legacy, shadingModel=MSM_Unlit, usedShadingModels=Unlit}
 nodes [
  Constant()
  TextureSampleParameter2D() {
   parameterName=Texture
   samplerType=SAMPLERTYPE_LinearColor
   texture=/Game/NeverCook/ArtTools/AnisotropicBaker/HairBake/TransientRendertargets/HairDilate01
  }
 ]
}
